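# Unlit translucent, two-sided material that draws the VR Editor snap grid, with parameterized color, center, rotation, interval and line width.
# World position is transformed into grid space, where the center axes are tinted red and green and fractional coordinates form line opacity.
# Opacity fades with distance from the object position toward the grid radius.
# Depth testing is off, so a manual pixel-versus-scene depth comparison leaves occluded lines faintly visible.
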Material /Engine/VREditor/SnapGrid/SnapGridMaterial
graph {
 properties {
  bDisableDepthTest=true
  bUsedWithSplineMeshes=true
  bUsedWithStaticLighting=true
  blendMode=BLEND_Translucent
  pixelDepthOffsetMode=PDOM_Legacy
  shadingModel=MSM_Unlit
  twoSided=true
  usedShadingModels=Unlit
 }
 nodes [
  n1 = VectorParameter() {parameterName=GridColor, defaultValue=<rgba 1 1 1 1>}
  n2 = WorldPosition()
  n3 = VectorParameter() {parameterName=GridCenter}
  n4 = ScalarParameter() {defaultValue=100, parameterName=GridInterval}
  n5 = Divide(
   A=MaterialFunctionCall(
    VectorToTransform=Subtract(A=@n2.XYZ, B=@n3.RGB)
    BasisX=VectorParameter() {parameterName=GridRotationMatrixX, defaultValue=<rgba 1 0 0 0>}.RGB
    BasisY=VectorParameter() {parameterName=GridRotationMatrixY, defaultValue=<rgba 0 1 0 0>}.RGB
    BasisZ=VectorParameter() {parameterName=GridRotationMatrixZ, defaultValue=<rgba 0 0 1 0>}.RGB
    BasisPosition=Constant3Vector()
   ) {materialFunction=&Transform3x3Matrix}.Transformed Vector
   B=@n4
  )
  n6 = Divide(A=Divide(A=ScalarParameter() {defaultValue=5, parameterName=LineWidth}, B=@n4))
  n7 = If(
   A=Abs(None=MaterialFunctionCall(Color=@n5) {materialFunction=&SplitComponents}.R)
   B=@n6
   "A > B"=Constant()
   "A < B"=Constant() {r=1}
  )
  n8 = If(
   A=Abs(None=MaterialFunctionCall(Color=@n5) {materialFunction=&SplitComponents}.G)
   B=@n6
   "A > B"=Constant()
   "A < B"=Constant() {r=1}
  )
  n9 = LinearInterpolate(A=@n1.RGB, Alpha=Max(A=@n7, B=@n8)) {constB=0}
  n10 = Constant3Vector() {constant=<rgba 1 0 0 0>}
  n11 = Constant3Vector() {constant=<rgba 0 1 0 0>}
  EyeAdaptationInverse(
   LightValueInput=Add(A=LinearInterpolate(A=@n9, B=@n10, Alpha=@n8), B=LinearInterpolate(A=@n9, B=@n11, Alpha=@n7))
   AlphaInput=Constant() {r=1}
  )
  n12 = MaterialFunctionCall(Color=Distance(A=ObjectPositionWS(), B=@n2.XYZ)) {materialFunction=&SplitComponents}
  n13 = Constant()
  n14 = ScalarParameter() {defaultValue=500, parameterName=GridRadius}
  n15 = Multiply(A=ScalarParameter() {defaultValue=0.5, parameterName=FalloffStartScale}, B=@n14) {desc=Falloff Start At}
  n16 = Subtract(A=@n14, B=@n15) {desc=Falloff Range}
  n17 = Clamp(
   None=Subtract(
    A=MaterialFunctionCall(X=@n12.R, Y=@n12.R, Z=@n13) {materialFunction=&MakeFloat3}.Result
    B=@n15
   )
   Max=@n16
  )
  n18 = Power(Base=OneMinus(None=Divide(A=@n17, B=@n16))) {constExponent=4}
  n19 = Frac(None=@n5)
  n20 = Clamp(None=Divide(A=@n19, B=@n6))
  n21 = Clamp(None=Divide(A=OneMinus(None=@n19), B=@n6))
  n22 = MaterialFunctionCall(Color=Add(A=OneMinus(None=@n20), B=OneMinus(None=@n21))) {materialFunction=&SplitComponents}
  n23 = Max(A=@n22.R, B=@n22.R)
  n24 = PixelDepth()
  n25 = SceneDepth()
  n26 = Constant() {desc=Opacity when occluded, r=0.05}
  n27 = Constant() {desc=Opacity when visible in plain sight, r=1}
  Multiply(A=@n1.A, B=Multiply(A=Multiply(A=@n18, B=@n23), B=If(A=@n24.R, B=@n25.R, "A > B"=@n26, "A < B"=@n27)))
 ]
}
